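# Editor Utility Widget that shows a quick details panel for the selected level actor.
# Tick binds that actor to a TextureData/Materials details view and its StaticMeshComponent to World Position Offset property views.
# UpdateMaterialsDetails copies four root-primitive material slots, resolving dynamic instances to parents, plus base materials, into widget variables shown by self-bound property views labelled Read Only.
# Without a primitive root, a switcher shows a warning instead.
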EditorUtilityWidgetBlueprint /FigmentTools/QuickDetailsView/EUW_QuickDetailsView
graphs [
 ClearMaterials {nodes [
   ClearMaterials() {is=title K2Node_FunctionEntry}
   > n1 = For Loop(LastIndex=3)
   > SetMaterialSlotProperty!(execute, self, Index:int=@n1, Visible:bool=false, Material:Material Interface Object Reference) -> (then)
  ]}
 EventGraph {nodes [
   Event PreConstruct()
   > SetPropertyName!(
    execute
    self:Single Property View Object Reference=Get SPV_Mat_0() -> (SPV_Mat_0:Single Property View Object Reference)
    NewPropertyName:Name=Mat0
   ) -> (then)
   > SetPropertyName(self=Get SPV_Mat_1() -> (SPV_Mat_1:Single Property View Object Reference), NewPropertyName=Mat1)
   > SetPropertyName(self=Get SPV_Mat_2() -> (SPV_Mat_2:Single Property View Object Reference), NewPropertyName=Mat2)
   > SetPropertyName(self=Get SPV_Mat_3() -> (SPV_Mat_3:Single Property View Object Reference), NewPropertyName=Mat3)
   > SetObject(self=[Get SPV_Mat_0(), Get SPV_Mat_1(), Get SPV_Mat_2(), Get SPV_Mat_3()], NewObject=Self-Reference())
   > SetPropertyName(self=Get SPV_Mat_0_Base() -> (SPV_Mat_0_Base:Single Property View Object Reference), NewPropertyName=Mat0Base)
   > SetPropertyName(self=Get SPV_Mat_1_Base() -> (SPV_Mat_1_Base:Single Property View Object Reference), NewPropertyName=Mat1Base)
   > SetPropertyName(self=Get SPV_Mat_2_Base() -> (SPV_Mat_2_Base:Single Property View Object Reference), NewPropertyName=Mat2Base)
   > SetPropertyName(self=Get SPV_Mat_3_Base() -> (SPV_Mat_3_Base:Single Property View Object Reference), NewPropertyName=Mat3Base)
   > SetObject(self=[Get SPV_Mat_3_Base(), Get SPV_Mat_0_Base(), Get SPV_Mat_1_Base(), Get SPV_Mat_2_Base()], NewObject=Self-Reference())
   > SetPropertyName(
    self=Get SinglePropertyView_WPO() -> (SinglePropertyView_WPO:Single Property View Object Reference)
    NewPropertyName=bEvaluateWorldPositionOffset
   )
   > SetPropertyName(
    self=Get SinglePropertyView_WPO_Dist() -> (SinglePropertyView_WPO_Dist:Single Property View Object Reference)
    NewPropertyName=WorldPositionOffsetDisableDistance
   )
   Event Tick()
   > n1 = GetSelectedLevelActors(self=Get EditorActorSubsystem())
   > n2 = Branch(Condition=Is Not Empty!(TargetArray:Array of Actor Object References=@n1) -> (ReturnValue:bool))
   > n3 = Set TargetActor(TargetActor:Actor Object Reference="Get (a copy)"(Array:Array of Actor Object References=@n1) -> (Output:Actor Object Reference)) -> (Output_Get:Actor Object Reference)
   > SetObject(self=Get DetailsView_TD() -> (DetailsView_TD:Details View Object Reference), NewObject=@n3)
   > SetObject(
    self=[Get SinglePropertyView_WPO_Dist(), Get SinglePropertyView_WPO()]
    NewObject=GetComponentByClass!(
     self:Actor Object Reference=Get TargetActor() -> (TargetActor:Actor Object Reference)
     ComponentClass:Actor Component Class Reference=&StaticMeshComponent
    ) -> (ReturnValue:Static Mesh Component Object Reference)
   )
   > n4 = UpdateMaterialsDetails!(execute, self) -> (then, bValidStats:bool)
   > n5 = Branch(Condition=@n4)
   n6 = Get WidgetSwitcher_State() -> (WidgetSwitcher_State:Widget Switcher Object Reference)
   > SetActiveWidgetIndex(self=@n6, Index=0)
   SetActiveWidgetIndex(execute=@n5.else, self=@n6, Index=1)
   SetObject(execute=@n2.else, self=Get DetailsView_TD())
  ]}
 SetMaterialSlotProperty {nodes [
   n1 = SetMaterialSlotProperty!() -> (then, Index:int, Visible:bool=~, Material:Material Interface Object Reference) {is=title K2Node_FunctionEntry}
   > SetVisibility(
    self=[
     Select!(
      Option 0:Single Property View Object Reference=Get SPV_Mat_0_Base()
      Option 1:Single Property View Object Reference=Get SPV_Mat_1_Base()
      Option 2:Single Property View Object Reference=Get SPV_Mat_2_Base()
      Option 3:Single Property View Object Reference=Get SPV_Mat_3_Base()
      Index:int=@n1
     ) -> (ReturnValue:Single Property View Object Reference)
     Select(Option 0=Get SPV_Mat_0(), Option 1=Get SPV_Mat_1(), Option 2=Get SPV_Mat_2(), Option 3=Get SPV_Mat_3(), Index=@n1)
    ]
    InVisibility=Select!(Option 0:ESlateVisibility Enum=Collapsed, Option 1:ESlateVisibility Enum=Visible, Index:bool=Get Visible() -> (Visible:bool)) -> (ReturnValue:ESlateVisibility Enum)
   )
   > n2 = Switch on Int!(execute:Exec, Selection:int=Get Index() -> (Index:int=0)) -> (Default:Exec, 0:Exec, 1:Exec, 2:Exec, 3:Exec)
   n3 = Get Material() -> (Material:Material Interface Object Reference)
   n4 = Cast To MaterialInstance!(Object=@n3) -> (AsMaterial Instance:Material Instance Object Reference, bSuccess:bool=~) {is=title K2Node_DynamicCast}
   n5 = Select(
    Option 0:Material Interface Object Reference=@n4
    Option 1:Material Interface Object Reference=Get Parent!(self:Material Instance Object Reference=@n4) -> (Parent:Material Interface Object Reference)
    Index=Cast To MaterialInstanceDynamic!(Object=@n3) -> (AsMaterial Instance Dynamic:Material Instance Dynamic Object Reference, bSuccess:bool=~) {is=title K2Node_DynamicCast}.bSuccess
   ) -> (ReturnValue:Material Interface Object Reference)
   n6 = Set Mat0(execute=@n2.0, Mat0:Material Interface Object Reference=@n5) -> (Output_Get:Material Interface Object Reference)
   > n7 = GetBaseMaterial!(execute, self:Material Interface Object Reference=@n6) -> (then, ReturnValue:Material Object Reference)
   > Set Mat0Base(Mat0Base:Material Interface Object Reference=@n7) -> (Output_Get:Material Interface Object Reference)
   n8 = Set Mat1(execute=@n2.1, Mat1:Material Interface Object Reference=@n5) -> (Output_Get:Material Interface Object Reference)
   > n9 = GetBaseMaterial(self=@n8)
   > Set Mat1Base(Mat1Base:Material Interface Object Reference=@n9) -> (Output_Get:Material Interface Object Reference)
   n10 = Set Mat2(execute=@n2.2, Mat2:Material Interface Object Reference=@n5) -> (Output_Get:Material Interface Object Reference)
   > n11 = GetBaseMaterial(self=@n10)
   > Set Mat2Base(Mat2Base:Material Interface Object Reference=@n11) -> (Output_Get:Material Interface Object Reference)
   n12 = Set Mat3(execute=@n2.3, Mat3:Material Interface Object Reference=@n5) -> (Output_Get:Material Interface Object Reference)
   > n13 = GetBaseMaterial(self=@n12)
   > Set Mat3Base(Mat3Base:Material Interface Object Reference=@n13) -> (Output_Get:Material Interface Object Reference)
   PrintString(execute=@n2.Default, InString=EUW Quick Details: Material Index is OOB)
  ]}
 UpdateMaterialsDetails {nodes [
   UpdateMaterialsDetails() {is=title K2Node_FunctionEntry}
   > n1 = Is Valid(InputObject=Get TargetActor())
   > n2 = Cast To PrimitiveComponent!(execute, Object=Get RootComponent!(self:Actor Object Reference=Get TargetActor()) -> (RootComponent:Scene Component Object Reference)) -> (then, CastFailed, AsPrimitive Component:Primitive Component Object Reference) {is=title K2Node_DynamicCast}
   > n3 = For Loop(LastIndex=3)
   Return Node!(execute=@n3.Completed, bValidStats:bool=true)
   n4 = GetMaterial!(self:Primitive Component Object Reference=@n2, ElementIndex:int=@n3) -> (ReturnValue:Material Interface Object Reference)
   > n5 = Is Valid(InputObject=@n4)
   SetMaterialSlotProperty(execute=@n5.Is Not Valid, Index=@n3, Visible=false) {is=title K2Node_CallFunction}
   SetMaterialSlotProperty(execute=@n5.Is Valid, Index=@n3, Visible=true, Material=@n4)
   ClearMaterials(execute=[@n2.CastFailed, @n1.Is Not Valid]) {is=title K2Node_CallFunction}
   > Return Node(bValidStats=false)
  ]}
]
variables [
 Mat0 {type=MaterialInterface, default=null}
 Mat0Base {type=MaterialInterface, default=null}
 Mat1 {type=MaterialInterface, default=null}
 Mat1Base {type=MaterialInterface, default=null}
 Mat2 {type=MaterialInterface, default=null}
 Mat2Base {type=MaterialInterface, default=null}
 Mat3 {type=MaterialInterface, default=null}
 Mat3Base {type=MaterialInterface, default=null}
 TargetActor {type=Actor, default=null}
]
widgets {
 info {rootWidgetClass=&Overlay, widgetCount=24}
 widgets [
  Overlay {widgetClassPath=&Overlay, widget=Overlay}
  WidgetSwitcher_State {
   widgetClassPath=&WidgetSwitcher
   bIsVariable=true
   parent=Overlay
   widget=WidgetSwitcher_State
   slot {class=OverlaySlot, properties {horizontalAlignment=HAlign_Fill, verticalAlignment=VAlign_Fill}}
  }
  EditorUtilityScrollBox_0 {
   widgetClassPath=&EditorUtilityScrollBox
   parent=WidgetSwitcher_State
   widget=EditorUtilityScrollBox_0
   slot {class=WidgetSwitcherSlot}
  }
  VerticalBox_0 {
   widgetClassPath=&VerticalBox
   parent=EditorUtilityScrollBox_0
   widget=VerticalBox_0
   properties {bOverride_Cursor=true}
   slot {class=ScrollBoxSlot}
  }
  VerticalBox_2 {
   widgetClassPath=&VerticalBox
   parent=VerticalBox_0
   widget=VerticalBox_2
   slot {class=VerticalBoxSlot, properties {padding=<ltrb 4 4 4 4>}}
  }
  TextBlock_79 {
   widgetClassPath=&TextBlock
   parent=VerticalBox_2
   widget=TextBlock_79
   properties {
    text=Parameters
    font {
     bForceMonospaced=false
     bMaterialIsStencil=false
     fontMaterial=null
     fontObject=&Roboto
     letterSpacing=0
     monospacedWidth=1
     size=12
     skewAmount=0
     typefaceFontName=Bold
     outlineSettings {
      bApplyOutlineToDropShadows=false
      bMiteredCorners=false
      bSeparateFillAlpha=false
      outlineMaterial=null
      outlineSize=0
      outlineColor=<rgba 0 0 0 1>
     }
    }
   }
   slot {class=VerticalBoxSlot, properties {padding=<ltrb 0 0 0 8>}}
  }
  SinglePropertyView_WPO {
   widgetClassPath=&SinglePropertyView
   parent=VerticalBox_2
   widget=SinglePropertyView_WPO
   slot {class=VerticalBoxSlot}
  }
  SinglePropertyView_WPO_Dist {
   widgetClassPath=&SinglePropertyView
   parent=VerticalBox_2
   widget=SinglePropertyView_WPO_Dist
   slot {class=VerticalBoxSlot}
  }
  DetailsView_TD {
   widgetClassPath=&DetailsView
   parent=VerticalBox_0
   widget=DetailsView_TD
   properties {propertiesToShow [TextureData,Materials]}
   slot {class=VerticalBoxSlot, properties {padding=<ltrb 4 4 4 4>}}
  }
  VerticalBox_3 {
   widgetClassPath=&VerticalBox
   parent=VerticalBox_0
   widget=VerticalBox_3
   slot {class=VerticalBoxSlot, properties {padding=<ltrb 4 4 4 4>}}
  }
  Border_0 {
   widgetClassPath=&Border
   parent=VerticalBox_3
   widget=Border_0
   properties {brushColor=<rgba 1 1 0 0.1>, padding=<ltrb 8 8 8 8>}
   slot {class=VerticalBoxSlot}
  }
  VerticalBox_4 {widgetClassPath=&VerticalBox, parent=Border_0, widget=VerticalBox_4, slot {class=BorderSlot, properties {padding=<ltrb 8 8 8 8>}}}
  TextBlock {
   widgetClassPath=&TextBlock
   parent=VerticalBox_4
   widget=TextBlock
   properties {
    text=Read Only
    font {
     bForceMonospaced=false
     bMaterialIsStencil=false
     fontMaterial=null
     fontObject=&Roboto
     letterSpacing=0
     monospacedWidth=1
     size=12
     skewAmount=0
     typefaceFontName=Bold
     outlineSettings {
      bApplyOutlineToDropShadows=false
      bMiteredCorners=false
      bSeparateFillAlpha=false
      outlineMaterial=null
      outlineSize=0
      outlineColor=<rgba 0 0 0 1>
     }
    }
   }
   slot {class=VerticalBoxSlot, properties {padding=<ltrb 0 0 0 8>}}
  }
  VerticalBox_1 {
   widgetClassPath=&VerticalBox
   parent=VerticalBox_4
   widget=VerticalBox_1
   properties {bOverride_Cursor=true}
   slot {class=VerticalBoxSlot, properties {padding=<ltrb 4 4 4 4>}}
  }
  SPV_Mat_0 {
   widgetClassPath=&SinglePropertyView
   parent=VerticalBox_1
   widget=SPV_Mat_0
   properties {nameOverride=Material Slot 1}
   slot {class=VerticalBoxSlot}
  }
  SPV_Mat_1 {
   widgetClassPath=&SinglePropertyView
   parent=VerticalBox_1
   widget=SPV_Mat_1
   properties {nameOverride=Material Slot 2}
   slot {class=VerticalBoxSlot}
  }
  SPV_Mat_2 {
   widgetClassPath=&SinglePropertyView
   parent=VerticalBox_1
   widget=SPV_Mat_2
   properties {nameOverride=Material Slot 3}
   slot {class=VerticalBoxSlot}
  }
  SPV_Mat_3 {
   widgetClassPath=&SinglePropertyView
   parent=VerticalBox_1
   widget=SPV_Mat_3
   properties {nameOverride=Material Slot 4}
   slot {class=VerticalBoxSlot}
  }
  VerticalBox {
   widgetClassPath=&VerticalBox
   parent=VerticalBox_4
   widget=VerticalBox
   properties {bOverride_Cursor=true}
   slot {class=VerticalBoxSlot, properties {padding=<ltrb 4 4 4 4>}}
  }
  SPV_Mat_0_Base {
   widgetClassPath=&SinglePropertyView
   parent=VerticalBox
   widget=SPV_Mat_0_Base
   properties {nameOverride=Material Slot 1 Base Mat}
   slot {class=VerticalBoxSlot}
  }
  SPV_Mat_1_Base {
   widgetClassPath=&SinglePropertyView
   parent=VerticalBox
   widget=SPV_Mat_1_Base
   properties {nameOverride=Material Slot 2 Base Mat}
   slot {class=VerticalBoxSlot}
  }
  SPV_Mat_2_Base {
   widgetClassPath=&SinglePropertyView
   parent=VerticalBox
   widget=SPV_Mat_2_Base
   properties {nameOverride=Material Slot 3 Base Mat}
   slot {class=VerticalBoxSlot}
  }
  SPV_Mat_3_Base {
   widgetClassPath=&SinglePropertyView
   parent=VerticalBox
   widget=SPV_Mat_3_Base
   properties {nameOverride=Material Slot 4 Base Mat}
   slot {class=VerticalBoxSlot}
  }
  TextBlock_63 {
   widgetClassPath=&TextBlock
   parent=WidgetSwitcher_State
   widget=TextBlock_63
   properties {
    text=Selected Actor does not have a Primitive Root
    font {
     bForceMonospaced=false
     bMaterialIsStencil=false
     fontMaterial=null
     fontObject=&Roboto
     letterSpacing=0
     monospacedWidth=1
     size=10.5
     skewAmount=0
     typefaceFontName=Bold
     outlineSettings {
      bApplyOutlineToDropShadows=false
      bMiteredCorners=false
      bSeparateFillAlpha=false
      outlineMaterial=null
      outlineSize=0
      outlineColor=<rgba 0 0 0 1>
     }
    }
   }
   slot {class=WidgetSwitcherSlot, properties {padding=<ltrb 4 4 4 4>}}
  }
 ]
}
